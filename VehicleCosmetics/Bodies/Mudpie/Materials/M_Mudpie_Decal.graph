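# Masked, Default Lit surface material for decals on the Mudpie vehicle cosmetic body.
# It exposes base color and normal texture parameters, defaulting to the Mudpie body decal textures, plus scalar Metallic, Specular and Roughness parameters.
# Usage is enabled for skeletal meshes and mesh deformers with automatic editor usage flagging off, and ray-traced shadow casting is disabled.
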Material /VehicleCosmetics/Bodies/Mudpie/Materials/M_Mudpie_Decal
graph {
 properties {
  bAutomaticallySetUsageInEditor=false
  bCastRayTracedShadows=false
  bUsedWithMeshDeformer=true
  bUsedWithSkeletalMesh=true
  blendMode=BLEND_Masked
  usedShadingModels=Default Lit
 }
 nodes [
  ScalarParameter() {parameterName=Metallic}
  ScalarParameter() {defaultValue=0.5, parameterName=Specular}
  ScalarParameter() {defaultValue=0.25, parameterName=Roughness}
  TextureSampleParameter2D() {parameterName=Decal_BC, texture=/VehicleCosmetics/Bodies/Mudpie/Textures/Body/T_Mudpie_Body_Decal_BC}
  TextureSampleParameter2D() {parameterName=Decal_N, samplerType=SAMPLERTYPE_Normal, texture=/VehicleCosmetics/Bodies/Mudpie/Textures/Body/T_Mudpie_Body_Decal_N}
 ]
}
