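# Body_Mudpie_PhysicsBR is the Physics Asset for the Mudpie vehicle body mesh in the VehicleCosmetics content.
# It defines a single body on the root bone, with no collision shapes listed.
# It has no constraints, so it sets up no jointed or ragdoll simulation between bones.
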PhysicsAsset /VehicleCosmetics/Bodies/Mudpie/Meshes/Body_Mudpie_PhysicsBR
bodies {root []}
constraints []
